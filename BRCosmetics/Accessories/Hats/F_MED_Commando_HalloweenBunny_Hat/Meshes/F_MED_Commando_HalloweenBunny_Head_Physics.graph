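# Physics asset for the Halloween Bunny female Commando hat's head mesh, defining capsule bodies and joints for dangling ears and hair.
# Both ears and the roots of three hair chains (left, right, f) attach to the head with limited swing and tight twist.
# Later hair segments constrain to the segment before them.
# The head, which carries two larger capsules, and the left clavicle link to spine_05 with free angular motion.
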PhysicsAsset /BRCosmetics/Accessories/Hats/F_MED_Commando_HalloweenBunny_Hat/Meshes/F_MED_Commando_HalloweenBunny_Head_Physics
bodies {
 clavicle_l [{shapeType=Capsule, length=0.505, radius=0.505, rotation=<rot 0 -90 0>}]
 dyn_ears_01_l [{shapeType=Capsule, length=1.1821, radius=3.1973, center=<11.1326 0.5103 0.4416>, rotation=<rot 0 -88.4501 -89.7779>}]
 dyn_ears_01_r [{shapeType=Capsule, length=1.1821, radius=3.1973, center=<11.2106 -0.5127 0.1362>, rotation=<rot 0 -91.5502 -89.7782>}]
 dyn_hair_01_l [{shapeType=Capsule, length=1.26, radius=1.25, center=<1.9045 0.0001 0>, rotation=<rot -0.0001 -89.9999 90.0003>}]
 dyn_hair_01_r [{shapeType=Capsule, length=2.9077, radius=1, center=<2.525 0 -0.0504>, rotation=<rot 83.9945 0.0001 0.0001>}]
 dyn_hair_02_l [{shapeType=Capsule, length=2.596, radius=0.75, center=<2.2385 0 0>, rotation=<rot 90 0 0>}]
 dyn_hair_02_r [{shapeType=Capsule, length=0.5, radius=1, center=<1.2795 0 0>, rotation=<rot 90 0 0>}]
 dyn_hair_03_r [{shapeType=Capsule, length=1.0748, radius=0.75, center=<1.0551 0 -1.022>, rotation=<rot 29.8453 0 0>}]
 dyn_hair_f_01 [{shapeType=Capsule, length=2.0414, radius=0.9414, center=<1.2076 0 0>, rotation=<rot 90 0 0>}]
 dyn_hair_f_02 [{shapeType=Capsule, length=0.7914, radius=0.6914, center=<0.2694 0 -0.0001>, rotation=<rot 90 0 0>}]
 head [
  {shapeType=Capsule, length=4.5, radius=5.2856, rotation=<rot 89.5203 -109.5179 -89.9894>}
  {shapeType=Capsule, length=15, radius=15, center=<13.7925 -14.5501 0.1454>, rotation=<rot -89.389 0.0042 74.0021>}
 ]
 spine_05 [{shapeType=Capsule, length=0.505, radius=0.505, rotation=<rot 0 -90 0>}]
}
constraints [
 {bone1Name=dyn_ears_01_l, bone2Name=head, swing1LimitDegrees=40, swing2LimitDegrees=35, twistLimitDegrees=5}
 {bone1Name=dyn_ears_01_r, bone2Name=head, swing1LimitDegrees=40, swing2LimitDegrees=35, twistLimitDegrees=5}
 {bone1Name=dyn_hair_01_l, bone2Name=head, swing1LimitDegrees=45, swing2LimitDegrees=45, twistLimitDegrees=5}
 {bone1Name=dyn_hair_02_l, bone2Name=dyn_hair_01_l, swing1LimitDegrees=45, swing2LimitDegrees=45, twistLimitDegrees=45}
 {bone1Name=dyn_hair_01_r, bone2Name=head, swing1LimitDegrees=45, swing2LimitDegrees=45, twistLimitDegrees=5}
 {bone1Name=dyn_hair_02_r, bone2Name=dyn_hair_01_r, swing1LimitDegrees=30, swing2LimitDegrees=30, twistLimitDegrees=5}
 {bone1Name=dyn_hair_03_r, bone2Name=dyn_hair_02_r, swing1LimitDegrees=45, swing2LimitDegrees=45, twistLimitDegrees=5}
 {bone1Name=dyn_hair_f_01, bone2Name=head, swing1LimitDegrees=45, swing2LimitDegrees=45, twistLimitDegrees=5}
 {bone1Name=dyn_hair_f_02, bone2Name=dyn_hair_f_01, swing1LimitDegrees=30, swing2LimitDegrees=30, twistLimitDegrees=5}
 {
  bone1Name=clavicle_l
  bone2Name=spine_05
  swing1LimitDegrees=45
  swing1Motion=Free
  swing2LimitDegrees=45
  swing2Motion=Free
  twistLimitDegrees=45
  twistMotion=Free
 }
 {
  bone1Name=head
  bone2Name=spine_05
  swing1LimitDegrees=45
  swing1Motion=Free
  swing2LimitDegrees=45
  swing2Motion=Free
  twistLimitDegrees=45
  twistMotion=Free
 }
]
