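# Animation Blueprint for the Halloween Bunny Commando hat, adding secondary physics motion to the cosmetic's head skeleton.
# It copies pose and curves from the attached parent mesh, then runs a partially blended RigidBody simulation in spine_05 space using the head physics asset.
# Three chained AnimDynamics nodes then add damped angular-spring sway to dyn_spine_tail bones 01, 03 and 04 at partial alpha.
# The event graph holds no logic.
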AnimBlueprint /BRCosmetics/Accessories/Hats/F_MED_Commando_HalloweenBunny_Hat/Meshes/F_MED_Commando_HalloweenBunny_Head_AnimBP
target_skeleton=&/F_MED_Commando_HalloweenBunny_Head_Skeleton
graphs [
 AnimGraph {nodes [
   n1 = RigidBody(ComponentPose=Local To Component(LocalPose=Copy Pose From Mesh() {bCopyCurves=true, bUseAttachedParent=true}), Alpha=0.45) {
    bOverrideWorldGravity=true
    overridePhysicsAsset=&/F_MED_Commando_HalloweenBunny_Head_Physics
    simulationSpace=BaseBoneSpace
    alphaBoolBlend {blendOption=Linear}
    baseBoneRef {boneName=spine_05}
    componentAppliedLinearAccClamp=<500 500 500>
    componentLinearAccScale=<1 1 1>
    componentLinearVelScale=<1 1 1>
    overrideWorldGravity {z=-980}
   }
   n2 = Anim Dynamics - Bone: dyn_spine_tail_01(ComponentPose=@n1, Alpha=0.65, ExternalForce="0,0,0") {
    angularDampingOverride=0.98
    angularSpringConstant=250
    bAngularSpring=true
    bGravityOverrideInSimSpace=true
    bOverrideAngularDamping=true
    gravityScale=0.25
    alphaBoolBlend {blendOption=Linear}
    boundBone {boneName=dyn_spine_tail_01}
    physicsBodyDefinitions [{
      sphereCollisionRadius=0
      boundBone {boneName=dyn_spine_tail_01}
      boxExtents=<30 30 30>
      constraintSetup {twistAxis=AxisY, angularLimitsMax=<5 2 30>, angularLimitsMin=<-5 -2 -30>, angularTarget=<0 0 0>}
      localJointOffset=<6 -1.25 -0.5>
     }]
   }
   n3 = Anim Dynamics - Bone: dyn_spine_tail_03(ComponentPose=@n2, Alpha=0.65, ExternalForce="0,0,0") {
    angularDampingOverride=0.99
    angularSpringConstant=275
    bAngularSpring=true
    bGravityOverrideInSimSpace=true
    bOverrideAngularDamping=true
    gravityScale=0.31
    alphaBoolBlend {blendOption=Linear}
    boundBone {boneName=dyn_spine_tail_03}
    physicsBodyDefinitions [{
      sphereCollisionRadius=0
      boundBone {boneName=dyn_spine_tail_03}
      boxExtents=<25 8 20>
      constraintSetup {twistAxis=AxisY, angularLimitsMax=<5 2 30>, angularLimitsMin=<-5 -2 -30>, angularTarget=<0 0 0>}
      localJointOffset=<1.5 5 0>
     }]
   }
   n4 = Anim Dynamics - Bone: dyn_spine_tail_04(ComponentPose=@n3, Alpha=0.25, ExternalForce="0,0,0") {
    angularDampingOverride=0.99
    angularSpringConstant=200
    bAngularSpring=true
    bGravityOverrideInSimSpace=true
    bOverrideAngularDamping=true
    gravityScale=0.3
    alphaBoolBlend {blendOption=Linear}
    boundBone {boneName=dyn_spine_tail_04}
    componentAppliedLinearAccClamp=<1000 1000 1000>
    componentLinearAccScale=<0.1 0.1 0.1>
    physicsBodyDefinitions [{
      sphereCollisionRadius=0
      boundBone {boneName=dyn_spine_tail_04}
      boxExtents=<15 8 10>
      constraintSetup {twistAxis=AxisY, angularLimitsMax=<5 2 30>, angularLimitsMin=<-5 -2 -30>, angularTarget=<0 0 0>}
      localJointOffset=<1 0 -2>
     }]
   }
   Output Pose(Result=Component To Local(ComponentPose=@n4)) {layerGroup=None}
  ]}
 EventGraph {nodes [Event BlueprintUpdateAnimation()]}
]
variables []
